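# Niagara system for the Lettuce emote that continuously throws tinted "Vbucks" mesh particles from two alternating bones or sockets on a user-supplied skeletal mesh.
# Six spawn partitions pick each particle's lifetime and bone, while color comes from a five-color palette.
# Particles launch outward with bone-aligned orientation, then tumble and drift under curl noise, aerodynamic drag and gravity.
# Curves stretch their Z scale at birth and shrink them away later.
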NiagaraSystem /BRCosmetics/Athena/Effects/Particles/Emotes/NS_Emote_Lettuce/NS_Emote_Lettuce_Main
dependencies {
 usedDataInterfaces [
  &NiagaraDataInterfaceArrayFloat
  &NiagaraDataInterfaceArrayColor
  &NiagaraDataInterfaceCurve
  &NiagaraDataInterfaceMeshRendererInfo
  &NiagaraDataInterfaceVectorCurve
 ]
 usedDynamicInputs [&Multiply_Vector2D, &MakeVector2D]
 usedModules [&SystemState]
}
emitters {Vbucks {
  emitterUpdateScript [
   &EmitterState()
   &SpawnRate(SpawnRate=60, Use Spawn Probability=false)
  ]
  particleSpawnScript [
   &PartitionParticles(Number Of Partitions=6)
   &InitializeParticle(
    Color=&SelectLinearColorFromArray(
     Array Sampling Mode=::NewEnumerator0
     Direct Array Index=?
     Color Selection Array=DataInterface() {propertyValues {ColorData [<rgba 0.9 0.81 0.225 1>, <rgba 1 0.45 0.5875 1>, <rgba 0.36 0.9 0.5535 1>, <rgba 0.8019 0.5 1 1>, <rgba 1 0.5127 0.3 1>]}}
     Random Seed=0
    )
    Lifetime=&SelectFloatFromArray(
     Direct Array Index=$Particles.PartitionParticles.Partition
     Direct Array Mode=::NewEnumerator1
     Float Selection Array=DataInterface() {propertyValues {FloatData [0.166,0.166,1.666,0.333,0.666,2]}}
    )
    Mesh Renderer Array Visibility Mode=::NewEnumerator3
    Mesh Renderer Info=DataInterface() {propertyValues {MeshRenderer=&self:Vbucks_2.NiagaraMeshRendererProperties_1}}
    Mesh Scale Mode=::NewEnumerator3
    UsePositionOffset=false
    Write Material Random=true
   )
   &SkeletalMeshLocation(
    Bone / Socket Index=&ModuloInt(A=$Particles.PartitionParticles.Partition, B=2)
    Bone Sampling Mode=::NewEnumerator4
    MaskSpawnGroup=false
    Orientation Sampling=::NewEnumerator1
    Sampled Position Offset=<0 3 0>
    Skeletal Mesh=$User.SkeletalMesh
   )
   &InitialMeshOrientation(
    Mesh Orientation Mode=::NewEnumerator6
    Orientation Quaternion:Quat4f=$Particles.SkeletalMeshLocation.SampledRotation
    Rotation=<0 0 0.25>
    Use Rotation Vector=true
   )
   &AddVelocity(
    Module.Use Falloff Distance=false
    Override Seed=false
    Velocity Mode=::NewEnumerator1
    Velocity Origin=$Output.SkeletalMeshLocation.SampledPositionWithoutOffset
    Velocity Speed=&RandomRangeFloat(Minimum=333, Maximum=666)
   )
   &ApplyInitialForces()
  ]
  particleUpdateScript [
   &ParticleState()
   &ScaleMeshSize(Scale Factor=&Multiply_VectorByFloat(
     Vector=&VectorFromCurve(
      VectorCurve=DataInterface() {propertyValues {XCurve {keys [<at 0 1>,<at 1 1>]}, YCurve {keys [<at 0 1>,<at 1 1>]}, ZCurve {keys [<at 0 0.1>,<at 0.4 1>]}}}
      CurveIndex=$Particles.Age
     )
     Float=&FloatFromCurve(FloatCurve=DataInterface() {propertyValues {Curve {keys [{tangentMode=RCTM_User, time=0.7, value=1}, {tangentMode=RCTM_User, time=1, value=0}]}}})
    ))
   &UpdateMeshOrientation(
    Rotation Rate=&RandomRangeFloat(Minimum=-1, Maximum=1)
    Rotation Vector=&Multiply_VectorByFloat(
     Vector=&RandomRangeVector(Minimum=<0 0 0>, Maximum=<1 1 1>)
     Float=&FloatFromCurve(FloatCurve=DataInterface() {propertyValues {Curve {keys [<at 0 0 RCIM_Cubic>,<at 0.1 1 RCIM_Cubic>]}}})
    )
   )
   &AerodynamicDrag(Ignore Mass=true, Surface Normal Axis:Vector3f=<0 0 1>, Use Linear Drag=true, Use Rotational Drag=true)
   /Niagara/Modules/Update/Forces/CurlNoiseForce(Noise Frequency=0.03, Noise Quality / Cost=::NewEnumerator0, Noise Strength=333, Pan Noise=true, Pan Noise Field=<0 0 1>)
   &GravityForce()
   &SolveForcesAndVelocity(Clamp Velocity=false, Limit Acceleration=false)
   &AlignSpriteToMeshOrientation(SetAlignment=true, SetFacing=true)
  ]
  rendererClasses [&NiagaraMeshRendererProperties]
  renderers [{rendererClass=&NiagaraMeshRendererProperties, rendererIndex=0}]
 }}
summary {
 systemName=NS_Emote_Lettuce_Main
 emitters [{emitterName=Vbucks, rendererClasses [&NiagaraMeshRendererProperties]}]
 userVariables [
  User.ForwardDot {type=NiagaraFloat, defaultValue=0, description=""}
  User.SkeletalMesh {
   type=NiagaraDataInterfaceSkeletalMesh
   defaultValue {
    struct=NiagaraExt_VariableValue_DataInterface
    value {dataInterface=&self:NiagaraDataInterfaceSkeletalMesh_0}
   }
  }
 ]
}
